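# Niagara burst effect for the Baba Yaga glider, combining bubbles, embers, flipbook flames, smoke, leaves and a short fading light.
# Bubbles, embers, flames and light spawn in bursts, smoke adds per-unit spawning, and leaves spawn only per unit moved.
# Every emitter samples User.SkeletalMesh and feeds bone position and rotation into a local TransformToBoneSpace module.
# Embers, flames and leaves take opposing owner velocity, and kill volumes confine bubbles, embers and flames.
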NiagaraSystem /Game/Gadgets/Assets/VinderTech_GliderChute/Glider_Baba_Yaga/Effects/NS_Glider_BabaYaga_Burst
dependencies {
 usedDataInterfaces [
  &NiagaraDataInterfaceCurve
  &NiagaraDataInterfaceVector2DCurve
  &NiagaraDataInterfaceColorCurve
  &NiagaraDataInterfaceVectorCurve
 ]
 usedDynamicInputs [&VectorFromCurve]
 usedModules [
  &self:SystemSpawnScript.SystemScriptSource.SystemScriptGraph.NiagaraNodeAssignment_0.SetVariables_#1
  &SystemState
 ]
}
emitters {
 Bubbles {
  emitterUpdateScript [
   &EmitterState()
   &SpawnBurst_Instantaneous(Spawn Count=&RandomRangeInt(Minimum=10, Maximum=20), Use Spawn Probability=false)
  ]
  particleSpawnScript [
   /SetVariables(
    Particles.OriginPoint:Vector3f=&Add_Vector(A=$System.OriginPoint, B=<0 0 22>)
    Particles.RandomPosNeg:NiagaraFloat=&UniformRangedFloat(Minimum=-1, Maximum=1)
   ) {isSetParameters=true}
   /Niagara/Modules/Spawn/Initialization/InitializeParticle(
    Color=<rgba 1 1 2 1>
    Lifetime=&RandomRangeFloat(Minimum=0.1, Maximum=0.6)
    Mesh Scale=&VectorFromFloat(Value=10)
    Module.Write Color=true
    Module.Write Lifetime=true
    Module.Write Mass=false
    Module.Write Material Random=true
    Module.Write Position=false
    Module.Write Scale=false
    Module.Write SpriteRotation=false
    Module.Write SpriteSize=true
    Position=&SimulationPosition()
    Sprite Size=&Vector2DFromFloat(Value=&RandomRangeFloat(Minimum=10, Maximum=30))
   )
   &CylinderLocation(
    Cylinder Height=0
    Cylinder Origin:Vector3f=$System.OriginPoint
    Cylinder Radius=10
    MaskSpawnGroup=false
    Module.Enable Lathe Profile=false
    Module.Override Randomness=false
    Module.Surface Only=false
    Offset=$Particles.OriginPoint
   ) {module=CylinderLocation001}
   /SetVariables(
    Particles.PositionBubble:NiagaraPosition=$Particles.Position
    Particles.SpriteSizeBubble:Vector2f=&Vector2DFromFloat(Value=&UniformRangedFloat(Minimum=0.5, Maximum=&RandomRangeFloat(Minimum=8, Maximum=16)))
    Particles.VelocityBubble:Vector3f=&UniformRangedVector(Minimum=<-10 -10 10>, Maximum=<10 10 100>)
   ) {isSetParameters=true}
   &CylinderLocation(
    Cylinder Height=0
    Cylinder Origin:Vector3f=$System.OriginPoint
    Cylinder Radius=10
    Lathe Profile=DataInterface() {propertyValues {Curve {keys [<at 0 0.451 RCIM_Cubic>, {arriveTangent=0.049, interpMode=RCIM_Cubic, leaveTangent=0.049, time=0.5, value=1}, <at 1 0.5 RCIM_Cubic>]}}}
    MaskSpawnGroup=false
    Module.Enable Lathe Profile=false
    Module.Override Randomness=false
    Module.Surface Only=false
    Offset=$Particles.OriginPoint
   )
   &AddVelocity(Velocity=&UniformRangedVector(Minimum=<-150 0 100>, Maximum=<150 150 200>))
  ]
  particleUpdateScript [
   &ParticleState()
   &SkeletalMeshLocation(
    Apply To Particle Position=false
    Bone Sampling Mode=::NewEnumerator4
    Inherit Velocity=false
    MaskSpawnGroup=false
    Skeletal Mesh=$User.SkeletalMesh
    Use Sampled Mesh Attribute as Ribbon ID=false
   )
   /SetVariables(Particles.BoneLocation:Vector3f=$Output.SkeletalMeshLocation.SampledPosition) {isSetParameters=true}
   &Drag(Drag=3, Use Linear Drag=true, Use Rotational Drag=false)
   &GravityForce()
   &SolveForcesAndVelocity(Clamp Velocity=false, Limit Acceleration=false)
   /Niagara/Modules/Update/Orientation/V2/UpdateMeshOrientation(Rotation Rate:NiagaraFloat=$Particles.RandomPosNeg, Rotation Vector:Vector3f=&VectorFromFloat(Value=1))
   &ScaleSpriteSize(Scale Factor=&Vector2DFromCurve(
     Vector2Curve=DataInterface() {propertyValues {
       XCurve {keys [<at 0 1>,<at 1 0.5>]}
       YCurve {keys [<at 0 0.1 RCIM_Cubic>, {arriveTangent=11.4975, interpMode=RCIM_Cubic, leaveTangent=11.4975, tangentMode=RCTM_User, time=1, value=7}]}
      }}
     ScaleCurve=&Vector2DFromFloat(Value=&Lerp_Float(A=1, B=0, Alpha=$Particles.NormalizedAge))
    ))
   &DynamicMaterialParameters(
    Index 0 Param 1=1
    Index 0 Param 2=1
    Index 0 Param 3=1
    Module.Param0WriteEnabled=true
    Module.Param1WriteEnabled=true
    Module.Param2WriteEnabled=true
    Module.Param3WriteEnabled=true
   )
   /SetVariables(
    Particles.PositionBubble:Vector3f=&Add_Vector(
     A=$Particles.Initial.PositionBubble
     B=&Multiply_VectorByFloat(
      Vector=$Particles.Initial.VelocityBubble
      Float=&Multiply_Float(A=$Particles.Lifetime, B=$Particles.NormalizedAge)
     )
    )
    Particles.SpriteSizeBubble:Vector2f=&Multiply_Vector2(
     A=$Particles.Initial.SpriteSizeBubble
     B=&Vector2DFromFloat(Value=&FloatFromCurve(
       FloatCurve=DataInterface() {propertyValues {Curve {keys [
           <at 0 0 RCIM_Cubic>
           {arriveTangent=0.002, interpMode=RCIM_Cubic, leaveTangent=-0.0489, tangentMode=RCTM_Break, time=0.125, value=0.9069}
           {arriveTangent=-0.077, interpMode=RCIM_Cubic, leaveTangent=-0.077, tangentMode=RCTM_User, time=0.45, value=0.8888}
           {arriveTangent=4.9648, interpMode=RCIM_Cubic, leaveTangent=4.9648, tangentMode=RCTM_User, time=0.5, value=1.2}
          ]}}}
       Scale Curve=&ClampFloat(Float=&Lerp_Float(A=20, B=-20, Alpha=$Particles.NormalizedAge))
      ))
    )
   ) {isSetParameters=true}
   &CameraOffset(Camera Offset Amount=10)
   &KillParticlesInVolume(Box Size:Vector3f=<100 100 200>, Invert Volume=true, Kill Shape=::NewEnumerator1, Origin Offset:Vector3f=<0 0 100>)
   &self:TransformToBoneSpace(
    BonePosition:Vector3f=$Particles.BoneLocation
    BoneRotation:Quat4f=$Output.SkeletalMeshLocation.SampledRotation
    Position:NiagaraPosition=$Particles.Position
   ) {module=TransformToBoneSpace}
   /SetVariables(Particles.PositionBubble:Vector3f=$Particles.OffsetPosition) {isSetParameters=true}
   &self:TransformToBoneSpace(
    BonePosition:Vector3f=$Particles.BoneLocation
    BoneRotation:Quat4f=$Output.SkeletalMeshLocation.SampledRotation
    Position:NiagaraPosition=$Particles.Position
   ) {module=TransformToBoneSpace001}
  ]
  rendererClasses [&NiagaraSpriteRendererProperties, &NiagaraMeshRendererProperties]
  renderers [
   {rendererClass=&NiagaraSpriteRendererProperties}
   {rendererClass=&NiagaraMeshRendererProperties, rendererIndex=1}
   {rendererClass=&NiagaraSpriteRendererProperties, rendererIndex=2}
  ]
 }
 Embers {
  emitterUpdateScript [
   &EmitterState()
   &SpawnBurst_Instantaneous(Spawn Count=20, Use Spawn Probability=false)
  ]
  particleSpawnScript [
   &InitializeParticle(
    Color=<rgba 2 1.2022 0.6482 1>
    Lifetime=&UniformRangedFloat(Minimum=0.1, Maximum=0.25)
    Module.Write Color=true
    Module.Write Lifetime=true
    Module.Write Mass=false
    Module.Write Material Random=true
    Module.Write Position=false
    Module.Write Scale=false
    Module.Write SpriteRotation=true
    Module.Write SpriteSize=true
    Position=&SimulationPosition()
    Sprite Rotation=&UniformRangedFloat(Minimum=-180, Maximum=180)
    Sprite Size=&Vector2DFromFloat(Value=&UniformRangedFloat(Minimum=1, Maximum=3))
   )
   &AddVelocity(Velocity=&RandomRangeVector(Minimum=<-100 -10 50>, Maximum=<100 10 400>))
   &ApplyInitialForces()
   &TorusLocation(
    Align Torus With Axis=false
    Handle Radius=2
    Large Radius=10
    MaskSpawnGroup=false
    Module.Override Randomness=false
    Torus Origin:NiagaraPosition=&Add_Vector(A=Linked() {linkedVariable {name=Particles.Position, type=NiagaraPosition}}, B=<0 -9 0>)
   )
  ]
  particleUpdateScript [
   &ParticleState()
   &SkeletalMeshLocation(
    Apply To Particle Position=false
    Bone Sampling Mode=::NewEnumerator4
    Inherit Velocity=false
    MaskSpawnGroup=false
    Skeletal Mesh=$User.SkeletalMesh
    Use Sampled Mesh Attribute as Ribbon ID=false
   )
   /SetVariables(Particles.BoneLocation:Vector3f=$Output.SkeletalMeshLocation.SampledPosition) {isSetParameters=true}
   &AddVelocity(Velocity=&RandomRangeVector(Minimum=<-50 5 -10>, Maximum=<50 40 10>, Evaluation Type=::NewEnumerator1)) {module=AddVelocity001}
   &InheritVelocity(
    Module.Limit Speed=true
    Velocity Limit=75
    Velocity Scale=<-0.1 -0.1 0>
    Velocity Source=&TransformVector(
     Vector=&RotateVectorByQuaternion(Vector To Rotate=$Engine.Owner.Velocity, Quaternion=$Particles.BoneOrient)
     Source Space=::World
     Destination Space=::Simulation
    )
   )
   &AccelerationForce(Acceleration=<0 0 50>)
   &Drag(Drag=3, Use Linear Drag=true, Use Rotational Drag=false)
   &SolveForcesAndVelocity(Clamp Velocity=false, Limit Acceleration=false)
   &ScaleColor(
    Scale Alpha=&FloatFromCurve(FloatCurve=DataInterface() {propertyValues {Curve {keys [
         {arriveTangent=11.9433, interpMode=RCIM_Cubic, leaveTangent=11.9433, tangentMode=RCTM_User, time=0, value=0}
         <at 0.2 1 RCIM_Cubic>
         <at 1 0 RCIM_Cubic>
        ]}}})
    Scale RGB=&MakeVectorFromLinearColorRGB(Color=&ColorFromCurve(ColorCurve=DataInterface() {propertyValues {
        AlphaCurve {keys [<at 0 1 RCIM_Cubic>]}
        BlueCurve {keys [<at 0 0.5069 RCIM_Cubic>,<at 1 0 RCIM_Cubic>]}
        GreenCurve {keys [<at 0 1.6422 RCIM_Cubic>,<at 1 0.06 RCIM_Cubic>]}
        RedCurve {keys [<at 0 4 RCIM_Cubic>,<at 1 1 RCIM_Cubic>]}
       }}))
    ScaleA=true
    ScaleRGB=true
   )
   &DynamicMaterialParameters(
    Index 0 Param 1=&FloatFromCurve(FloatCurve=DataInterface() {propertyValues {Curve {keys [<at 0 0>,<at 1 1>]}}})
    Index 0 Param 2=1
    Index 0 Param 3=1
    Index 0 Param 4=1
    Module.Param0WriteEnabled=true
    Module.Param1WriteEnabled=true
    Module.Param2WriteEnabled=true
    Module.Param3WriteEnabled=true
   )
   &ScaleSpriteSize(Scale Factor=&Vector2DFromFloat(Value=&FloatFromCurve(FloatCurve=DataInterface() {propertyValues {Curve {keys [
          {arriveTangent=9.1531, interpMode=RCIM_Cubic, leaveTangent=9.1531, tangentMode=RCTM_User, time=0, value=0.02}
          {arriveTangent=0.48, interpMode=RCIM_Cubic, leaveTangent=0.48, time=0.18, value=0.99}
          <at 1 0.5 RCIM_Cubic>
         ]}}})))
   &ScaleSpriteSizeBySpeed(Max Scale Factor=<1 4>)
   &KillParticlesInVolume(
    Box Size:Vector3f=<50 500 100>
    Invert Volume=true
    Kill Shape=::NewEnumerator1
    Origin Offset:Vector3f=<0 100 0>
    Volume Origin:Vector3f=$Particles.OriginPoint
   )
   &self:TransformToBoneSpace(
    BonePosition:Vector3f=$Particles.BoneLocation
    BoneRotation:Quat4f=$Output.SkeletalMeshLocation.SampledRotation
    Position:NiagaraPosition=$Particles.Position
   ) {module=TransformToBoneSpace}
  ]
  rendererClasses [&NiagaraSpriteRendererProperties]
  renderers [{rendererClass=&NiagaraSpriteRendererProperties}]
 }
 Flames {
  emitterUpdateScript [
   &EmitterState()
   &SpawnBurst_Instantaneous(Spawn Count=10, Use Spawn Probability=false)
  ]
  particleSpawnScript [
   &InitializeParticle(
    Color=/Niagara/DynamicInputs/TypeConversions/MakeLinearColorFromFloat(Value:NiagaraFloat=&RandomRangeFloat(Minimum=0.25, Maximum=1))
    Lifetime=&UniformRangedFloat(Minimum=0.1, Maximum=&RandomRangeFloat(Minimum=0.3, Maximum=0.5))
    Mesh Scale=&VectorFromFloat(Value=3)
    Module.Write Color=true
    Module.Write Lifetime=true
    Module.Write Mass=false
    Module.Write Material Random=true
    Module.Write Position=false
    Module.Write Scale=true
    Module.Write SpriteRotation=true
    Module.Write SpriteSize=true
    Position=&SimulationPosition()
    Sprite Rotation=&UniformRangedFloat(Minimum=-180, Maximum=180)
    Sprite Size=&Vector2DFromFloat(Value=&UniformRangedFloat(Minimum=20, Maximum=50))
   )
   /SetVariables(Particles.OriginPoint:Vector3f=&Add_Vector(A=$System.OriginPoint, B=<0 0 2>)) {isSetParameters=true}
   &TorusLocation(
    Align Torus With Axis=false
    Handle Radius=2
    Large Radius=10
    MaskSpawnGroup=false
    Module.Override Randomness=false
    Torus Origin:Vector3f=$Particles.OriginPoint
   )
   &DynamicMaterialParameters(
    Index 0 Param 2=&UniformRangedFloat(Minimum=0, Maximum=1)
    Index 0 Param 3=1
    Index 0 Param 4=&UniformRangedFloat(Minimum=1, Maximum=2)
    Module.Param0WriteEnabled=false
    Module.Param1WriteEnabled=true
    Module.Param2WriteEnabled=false
    Module.Param3WriteEnabled=true
   ) {module=DynamicMaterialParameters002}
   &AddVelocityFromPoint(Module.Use Falloff Distance=false, Origin Offset=$Particles.OriginPoint, Velocity Strength=200)
   /SetVariables(
    Particles.ParticleRandom:NiagaraFloat=&RandomRangeFloat(Minimum=0, Maximum=1)
    Particles.Velocity=&Multiply_Vector(A=$Particles.Velocity, B=<1 1 0.1>)
   ) {isSetParameters=true}
  ]
  particleUpdateScript [
   &ParticleState()
   &SkeletalMeshLocation(
    Apply To Particle Position=false
    Bone Sampling Mode=::NewEnumerator4
    Inherit Velocity=false
    MaskSpawnGroup=false
    Skeletal Mesh=$User.SkeletalMesh
    Use Sampled Mesh Attribute as Ribbon ID=false
   )
   /SetVariables(Particles.BoneLocation:Vector3f=$Output.SkeletalMeshLocation.SampledPosition) {isSetParameters=true}
   &InheritVelocity(
    Module.Limit Speed=true
    Velocity Limit=8
    Velocity Scale=<-0.01 -0.01 0>
    Velocity Source=&TransformVector(Vector=$Engine.Owner.Velocity, Source Space=::World, Destination Space=::Simulation)
   )
   &AccelerationForce(Acceleration=&UniformRangedVector(Minimum=<0 0 100>, Maximum=<1 1 600>))
   &Drag(Drag=3.5, Use Linear Drag=true, Use Rotational Drag=false)
   &SolveForcesAndVelocity(Clamp Velocity=false, Limit Acceleration=false)
   &ScaleColor(
    Scale Mode=::NewEnumerator1
    Scale RGBA=/Niagara/DynamicInputs/UniformRange/UniformRangedVector4(
     Minimum:Vector4f=&Vector4FromCurve(Vector4Curve=DataInterface() {propertyValues {
        AlphaCurve {keys [<at 0.0024 0>,<at 0.1271 1>,<at 0.265 1>,<at 0.9977 -0.0026>]}
        BlueCurve {keys [<at 0.2027 0.5>,<at 0.4709 0.1275>,<at 0.75 0>]}
        GreenCurve {keys [<at 0.2027 2.875>,<at 0.4709 0.7331>,<at 0.75 0>]}
        RedCurve {keys [<at 0.2027 10>,<at 0.4709 5>,<at 0.75 2>]}
       }})
     Maximum:Vector4f=&Vector4FromCurve(Vector4Curve=DataInterface() {propertyValues {
        AlphaCurve {keys [<at 0.0024 0>,<at 0.1271 1>,<at 0.265 0.9976>,<at 0.9977 -0.0026>]}
        BlueCurve {keys [<at 0.3937 0.5>,<at 0.75 0>]}
        GreenCurve {keys [<at 0.3937 2.875>,<at 0.75 0.5>]}
        RedCurve {keys [<at 0.3937 10>,<at 0.75 10>]}
       }})
     Randomness Mode:ENiagaraRandomnessMode=::NewEnumerator0
     Random Seed:NiagaraInt32=0
     Fixed Random Seed:NiagaraBool=false
    )
    ScaleRGBA=true
   ) {module=ScaleColor001}
   /Niagara/Modules/Update/SubUV/SubUVAnimation(Number Of Frames=64, SubUV Lookup Index=&Add_Float(A=$Particles.ParticleRandom, B=$Particles.Age))
   &ScaleSpriteSize(Scale Factor=&Vector2DFromFloat(Value=&FloatFromCurve(FloatCurve=DataInterface() {propertyValues {Curve {keys [
          {arriveTangent=10.0177, interpMode=RCIM_Cubic, leaveTangent=10.0177, tangentMode=RCTM_User, time=0, value=0}
          {arriveTangent=1.2063, interpMode=RCIM_Cubic, leaveTangent=1.2063, tangentMode=RCTM_User, time=0.1683, value=1.5622}
          <at 1 0.5 RCIM_Cubic>
         ]}}})))
   &CameraOffset(Camera Offset Amount=5)
   &DynamicMaterialParameters(
    Index 0 Param 1=&FloatFromCurve(FloatCurve=DataInterface() {propertyValues {Curve {keys [{arriveTangent=2.7551, interpMode=RCIM_Cubic, leaveTangent=2.7551, tangentMode=RCTM_User, time=0, value=0}, <at 0.7 1>]}}})
    Index 0 Param 2=1
    Index 0 Param 3=1
    Index 0 Param 4=1
    Module.Param0WriteEnabled=true
    Module.Param1WriteEnabled=true
    Module.Param2WriteEnabled=true
    Module.Param3WriteEnabled=false
   ) {module=DynamicMaterialParameters001}
   &KillParticlesInVolume(Invert Volume=true, Sphere Radius=40, Volume Origin:Vector3f=$Particles.OriginPoint)
   &self:TransformToBoneSpace(
    BonePosition:Vector3f=$Particles.BoneLocation
    BoneRotation:Quat4f=$Output.SkeletalMeshLocation.SampledRotation
    Position:Vector3f=$Particles.LightPosition
   ) {module=TransformToBoneSpace}
  ]
  rendererClasses [&NiagaraSpriteRendererProperties]
  renderers [{rendererClass=&NiagaraSpriteRendererProperties}]
 }
 Leaf_GPU {
  emitterUpdateScript [
   &EmitterState(Life Cycle Mode=::NewEnumerator1, Loop Behavior=::NewEnumerator1, Loop Delay=0.25, Loop Duration=1, UseLoopDelay=true)
   &SpawnPerUnit(
    Module.Use Max Movement Threshold=true
    Module.Use Movement Tolerance=false
    Spawn Per Unit:NiagaraFloat=&Lerp_Float(A=5, B=500, Alpha=$Emitter.NormalizedLoopAge)
    Use Spawn Probability=false
   )
  ]
  particleSpawnScript [
   &InitializeParticle(
    Module.Write Color=false
    Module.Write Lifetime=true
    Module.Write Mass=false
    Module.Write Material Random=true
    Module.Write Position=false
    Module.Write Scale=false
    Module.Write SpriteRotation=true
    Module.Write SpriteSize=true
    Position=&SimulationPosition()
    Sprite Rotation=&UniformRangedFloat(Minimum=0, Maximum=360)
    Sprite Size=&UniformRangedVector2D(Minimum=<5 10>, Maximum=<10 20>)
   )
   /SetVariables(
    Particles.SpriteFacing=&UniformRangedVector(Minimum=<-1 -1 -1>, Maximum=<1 1 1>)
    Particles.SpriteInitRotationRate:NiagaraFloat=&Multiply_Float(A=&UniformAOrBFloat(A=-1), B=&UniformRangedFloat(Minimum=1, Maximum=2))
   ) {isSetParameters=true}
   &AddVelocity(Velocity=&UniformRangedVector(Minimum=<-100 -50 0.75>, Maximum=<100 100 1>))
   &ApplyInitialForces()
   &InheritVelocity(
    Module.Limit Speed=true
    Velocity Limit=1000
    Velocity Scale=<-1 -1 0>
    Velocity Source=&TransformVector(Vector=$Engine.Owner.Velocity, Source Space=::World, Destination Space=::Simulation)
   )
   &SkeletalMeshLocation(Bone / Socket Index=1, Bone Sampling Mode=::NewEnumerator4, Inherit Velocity=false, MaskSpawnGroup=false, Skeletal Mesh=$User.SkeletalMesh)
  ]
  particleUpdateScript [
   &ParticleState()
   &SkeletalMeshLocation(
    Apply To Particle Position=false
    Bone Sampling Mode=::NewEnumerator4
    Inherit Velocity=false
    MaskSpawnGroup=false
    Skeletal Mesh=$User.SkeletalMesh
    Use Sampled Mesh Attribute as Ribbon ID=false
   ) {module=SkeletalMeshLocation001}
   /SetVariables(Particles.BoneLocation:Vector3f=$Output.SkeletalMeshLocation001.SampledPosition) {isSetParameters=true}
   &ScaleSpriteSize(Scale Factor=&MakeVector2D(
     X=&FloatFromCurve(FloatCurve=DataInterface() {propertyValues {Curve {keys [
          {interpMode=RCIM_Cubic, tangentMode=RCTM_User, time=0, value=0}
          <at 0.1 1>
          {arriveTangent=0.002, interpMode=RCIM_Cubic, leaveTangent=0.002, time=0.5, value=1}
          <at 1.0011 0.002>
         ]}}})
     Y=1
    ))
   &ScaleVelocity(Velocity Scale=&Lerp_Vector(A=<1 1 50>, B=<1 1 1000>, Alpha=$Particles.NormalizedAge))
   &SolveForcesAndVelocity(Clamp Velocity=false, Limit Acceleration=false)
   &SpriteRotationRate(Rotation Rate=&Multiply_Float(A=$Particles.SpriteInitRotationRate, B=&VectorLength(Vector=$Particles.Velocity)))
   /SetVariables(Particles.SpriteFacing=&Add_Vector(
     A=$Particles.Initial.SpriteFacing
     B=&Multiply_VectorByFloat(
      Vector=/Niagara/DynamicInputs/Helpers/SafeNormalizeVector(Vector To Normalize:Vector3f=Linked() {linkedVariable {name=Particles.Position, type=NiagaraPosition}}, Fallback Vector:Vector3f=<0 0 0>)
      Float=0.5
     )
    )) {isSetParameters=true}
   &SolveRotationalForcesAndVelocity()
   &self:TransformToBoneSpace(
    BonePosition:Vector3f=$Particles.BoneLocation
    BoneRotation:Quat4f=$Output.SkeletalMeshLocation001.SampledRotation
    Position:Vector3f=$Particles.LightPosition
   ) {module=TransformToBoneSpace}
  ]
  rendererClasses [&NiagaraSpriteRendererProperties]
  renderers [{rendererClass=&NiagaraSpriteRendererProperties}]
 }
 Light {
  emitterUpdateScript [
   &EmitterState()
   &SpawnBurst_Instantaneous(Spawn Count=1, Use Spawn Probability=false)
  ]
  particleSpawnScript [
   &InitializeParticle(
    Color=<rgba 6 1.8281 0.4375 1>
    Lifetime=0.5
    Module.Write Color=true
    Module.Write Lifetime=true
    Module.Write Mass=false
    Module.Write Material Random=true
    Module.Write Position=false
    Module.Write Scale=false
    Module.Write SpriteRotation=false
    Module.Write SpriteSize=true
    Position=&SimulationPosition()
    Sprite Size=<60 40>
   )
   &DynamicMaterialParameters(
    Index 0 Param 1=-0.5
    Index 0 Param 2=50
    Module.Param0WriteEnabled=true
    Module.Param1WriteEnabled=true
    Module.Param2WriteEnabled=true
    Module.Param3WriteEnabled=true
   )
   &SystemLocation(MaskSpawnGroup=false, Offset=<0 -9 10>)
  ]
  particleUpdateScript [
   &SkeletalMeshLocation(
    Apply To Particle Position=false
    Bone Sampling Mode=::NewEnumerator4
    Inherit Velocity=false
    MaskSpawnGroup=false
    Skeletal Mesh=$User.SkeletalMesh
    Use Sampled Mesh Attribute as Ribbon ID=false
   )
   /SetVariables(Particles.BoneLocation:Vector3f=$Output.SkeletalMeshLocation.SampledPosition) {isSetParameters=true}
   /SetVariables(
    Particles.LightExponent=1
    Particles.LightPosition:Vector3f=&Add_Vector(A=Linked() {linkedVariable {name=Particles.Position, type=NiagaraPosition}}, B=<0 0 10>)
    Particles.LightRadius=75
   ) {isSetParameters=true}
   &ParticleState()
   &SolveForcesAndVelocity(Clamp Velocity=false, Limit Acceleration=false)
   &ScaleColor(
    Scale RGB=&VectorFromFloat(Value=&Lerp_Float(A=1, B=0, Alpha=$Particles.NormalizedAge))
    ScaleA=false
    ScaleRGB=true
   )
   &self:TransformToBoneSpace(
    BonePosition:Vector3f=$Particles.BoneLocation
    BoneRotation:Quat4f=$Output.SkeletalMeshLocation.SampledRotation
    Position:Vector3f=$Particles.LightPosition
   ) {module=TransformToBoneSpace}
  ]
  rendererClasses [&NiagaraLightRendererProperties]
  renderers [{rendererClass=&NiagaraLightRendererProperties, rendererIndex=0}]
 }
 Smoke {
  emitterUpdateScript [
   &EmitterState()
   &SpawnBurst_Instantaneous(Spawn Count=15, Use Spawn Probability=false)
   &SpawnPerUnit(
    Module.Use Max Movement Threshold=true
    Module.Use Movement Tolerance=false
    Spawn Per Unit:NiagaraFloat=&Lerp_Float(A=15, B=200, Alpha=$Emitter.NormalizedLoopAge)
    Use Spawn Probability=false
   )
  ]
  particleSpawnScript [
   &InitializeParticle(
    Color=&UniformRangedLinearColor(Minimum=<rgba 0.171 0.171 0.171 0.5>, Maximum=<rgba 0.228 0.228 0.228 0.95>)
    Lifetime=&UniformRangedFloat(Minimum=0.25, Maximum=0.5)
    Module.Write Color=true
    Module.Write Lifetime=true
    Module.Write Mass=true
    Module.Write Material Random=true
    Module.Write Position=false
    Module.Write Scale=false
    Module.Write SpriteRotation=true
    Module.Write SpriteSize=true
    Position=&SimulationPosition()
    Sprite Rotation=&UniformRangedFloat(Minimum=0, Maximum=360)
    Sprite Size=&Vector2DFromFloat(Value=&UniformRangedFloat(Minimum=10, Maximum=50))
   )
   &AccelerationForce(Acceleration=&UniformRangedVector(Minimum=<0 0 0>, Maximum=<0 0 50>))
   &ApplyInitialForces()
   &SkeletalMeshLocation(
    Inherit Velocity=false
    MaskSpawnGroup=false
    Mesh Sampling Type=::NewEnumerator5
    Skeletal Mesh=$User.SkeletalMesh
    Socket Sampling Mode=::NewEnumerator5
   )
   &AddVelocity(Velocity=&RandomRangeVector(Minimum=<-150 -150 -150>, Maximum=<150 150 150>))
   &InheritVelocity(Module.Limit Speed=false, Velocity Scale=&VectorFromFloat(Value=1))
  ]
  particleUpdateScript [
   &ParticleState()
   &SkeletalMeshLocation(
    Apply To Particle Position=false
    Bone Sampling Mode=::NewEnumerator4
    Inherit Velocity=false
    MaskSpawnGroup=false
    Skeletal Mesh=$User.SkeletalMesh
    Use Sampled Mesh Attribute as Ribbon ID=false
   ) {module=SkeletalMeshLocation001}
   /SetVariables(Particles.BoneLocation:Vector3f=$Output.SkeletalMeshLocation001.SampledPosition) {isSetParameters=true}
   &Drag(Drag=3, Use Linear Drag=true, Use Rotational Drag=false)
   &SolveForcesAndVelocity(Clamp Velocity=false, Limit Acceleration=false)
   &ScaleSpriteSize(Scale Factor=&Vector2DFromFloat(Value=&FloatFromCurve(
      FloatCurve=DataInterface() {propertyValues {Curve {keys [{arriveTangent=2.1794, interpMode=RCIM_Cubic, leaveTangent=2.1794, tangentMode=RCTM_User, time=0, value=0}, <at 1 1 RCIM_Cubic>]}}}
      Scale Curve=2
     )))
   /SetVariables(Particles.SubImageIndex=&FloatFromCurve(FloatCurve=DataInterface() {propertyValues {Curve {keys [
         {arriveTangent=29.8623, interpMode=RCIM_Cubic, leaveTangent=29.8623, tangentMode=RCTM_User, time=-0.0046, value=5.0556}
         <at 1.0011 22.1689 RCIM_Cubic>
        ]}}})) {isSetParameters=true}
   &DynamicMaterialParameters(
    Index 0 Param 1=&FloatFromCurve(FloatCurve=DataInterface() {propertyValues {Curve {keys [<at -0.0046 1.0113 RCIM_Cubic>,<at 1.0011 0.002>]}}})
    Index 0 Param 4=&FloatFromCurve(FloatCurve=DataInterface() {propertyValues {Curve {keys [<at 0 0 RCIM_Cubic>,<at 1 0.5>]}}})
    Module.Param0WriteEnabled=true
    Module.Param1WriteEnabled=false
    Module.Param2WriteEnabled=false
    Module.Param3WriteEnabled=true
   ) {module=DynamicMatParams}
   &ScaleColor(
    Scale Alpha=&FloatFromCurve(FloatCurve=DataInterface() {propertyValues {Curve {keys [
         {arriveTangent=-3.1757, interpMode=RCIM_Cubic, leaveTangent=-3.1757, tangentMode=RCTM_User, time=0.5, value=1.0113}
         {arriveTangent=-1.5121, interpMode=RCIM_Cubic, leaveTangent=-1.5121, tangentMode=RCTM_User, time=1.0011, value=0.002}
        ]}}})
    ScaleA=true
    ScaleRGB=false
   )
   &self:TransformToBoneSpace(
    BonePosition:Vector3f=$Particles.BoneLocation
    BoneRotation:Quat4f=$Output.SkeletalMeshLocation001.SampledRotation
    Position:Vector3f=$Particles.LightPosition
   ) {module=TransformToBoneSpace}
  ]
  rendererClasses [&NiagaraSpriteRendererProperties]
  renderers [{rendererClass=&NiagaraSpriteRendererProperties}]
 }
}
summary {
 systemName=NS_Glider_BabaYaga_Burst
 emitters [
  {emitterName=Bubbles, rendererClasses [&NiagaraSpriteRendererProperties, &NiagaraMeshRendererProperties]}
  {emitterName=Embers, rendererClasses [&NiagaraSpriteRendererProperties]}
  {emitterName=Light, rendererClasses [&NiagaraLightRendererProperties]}
  {emitterName=Flames, rendererClasses [&NiagaraSpriteRendererProperties]}
  {emitterName=Smoke, rendererClasses [&NiagaraSpriteRendererProperties]}
  {emitterName=Leaf_GPU, rendererClasses [&NiagaraSpriteRendererProperties]}
 ]
 userVariables [
  User.SkeletalMesh {
   type=NiagaraDataInterfaceSkeletalMesh
   defaultValue {
    struct=NiagaraExt_VariableValue_DataInterface
    value {dataInterface=&self:NiagaraDataInterfaceSkeletalMesh_0}
   }
  }
  User.TrailLengthScale {type=NiagaraFloat, defaultValue=0, description=""}
 ]
}
